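# Character material for the CoyoTear backpack cosmetic that adds animated, flickering emissive.
# It wraps custom logic between the shared 2023 CharacterParent PreFX and PostFX functions, overriding only Emissive Color.
# Time-driven noise modulates the emissive, with one side using simple noise and the other a stepped Voronoi flicker.
# A step on pre-skinned local X, passed through a vertex interpolator, picks the pattern per side, so the split stays fixed under animation.
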Material /BRCosmetics/Accessories/FORT_Backpacks/Backpack_F_MED_CoyoTear/Materials/M_F_MED_CoyoTear_Backpack
graph {
 properties {
  bUseMaterialAttributes=true
  bUsedWithMeshDeformer=true
  bUsedWithSkeletalMesh=true
  pixelDepthOffsetMode=PDOM_Legacy
  shadingModel=MSM_FromMaterialExpression
  usedShadingModels=Default Lit | Subsurface | Clear Coat | Subsurface Profile | Cloth
 }
 nodes [
  n1 = MaterialFunctionCall() {materialFunction=&MF_CharacterParent_2023_PreFX}
  Composite!(None=@n1.Result)
  MaterialExpressionReroute_1 = Reroute(None=@n1.Result)
  n2 = GetMaterialAttributes(None=@MaterialExpressionReroute_1.Input) {attributeGetTypes [#1]}
  n3 = NamedRerouteUsage()
  n4 = Noise!(World Position=Time(), FilterWidth) {levels=1, outputMin=0}
  n5 = NamedRerouteUsage()
  n6 = Time()
  n7 = Multiply(A=@n6) {constB=0.3}
  n8 = LinearInterpolate(
   A=Noise(World Position=Multiply(A=@n6) {constB=0.8}) {levels=1, outputMin=0.3}
   Alpha=Step(X=Noise(World Position=@n7) {levels=1, noiseFunction=NOISEFUNCTION_VoronoiALU, outputMin=0}) {constY=0.3}
  )
  n9 = ComponentMask(None=VertexInterpolator(VS=PreSkinnedPosition()).PS) {r=true}
  n10 = SetMaterialAttributes(
   MaterialAttributes=@n2.MaterialAttributes
   Emissive Color=LinearInterpolate(A=Multiply(A=@n3, B=@n4), B=Multiply(A=@n5, B=@n8), Alpha=Step(X=@n9))
  ) {attributeSetTypes [#1]}
  MaterialExpressionReroute_0 = Reroute(None=@n10)
  MaterialFunctionCall(Material=@MaterialExpressionReroute_0.Output, M=@n1.M, Pre Skinned Local Position=@n1.M, Distance Blend=@n1.M, Fresnel=@n1.M) {materialFunction=&MF_CharacterParent_2023_PostFX}
  NamedRerouteDeclaration(None=@n2.Emissive Color) {name=Emissive, nodeColor=<rgba 0 1 0.8828 1>}
  PinBase() {reroutePins [Input {expression=&self:MaterialExpressionReroute_1}]}
  PinBase!(Output=@n10) {reroutePins [Output {expression=&self:MaterialExpressionReroute_0}]}
 ]
}
