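# Physics asset for the CoyoTear backpack cosmetic, simulating a dangling tiger head and two rope chains.
# Capsule bodies cover the dynamic bones, the pack, and character pelvis, spine and right upper arm bones, likely as collision shapes for the dangles.
# The head and three-segment rope A hang from the pack on free angular joints.
# Constraints with set swing and twist limits link the rope segments and attach two-segment rope B to the pack.
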PhysicsAsset /BRCosmetics/Accessories/FORT_Backpacks/Backpack_F_MED_CoyoTear/Meshes/F_MED_CoyoTear_Pack_Physics
bodies {
 dyn_Tiger_Head_02 [{shapeType=Capsule, length=8, radius=11, shapeName=dyn_Tiger_Head_02_capsule, center=<0 -0.924 5.3846>, rotation=<rot 90 -0.0789 -0.0791>}]
 dyn_rope_a01 [{
   shapeType=Capsule
   length=3.2689
   radius=2.9821
   shapeName=dyn_rope_a01_capsule
   center=<-1.0239 0.2213 -5.0732>
   rotation=<rot -4.9901 -89.7373 -1.507>
  }]
 dyn_rope_a02 [{shapeType=Capsule, length=8.1605, radius=2.1402, shapeName=dyn_rope_a02_capsule, center=<0 -0.5 -4.9507>, rotation=<rot 8.2905 -90 2.1508>}]
 dyn_rope_a03 [{
   shapeType=Capsule
   length=9.4942
   radius=2.3623
   shapeName=dyn_rope_a03_capsule
   center=<0.3517 -1.6789 -5.5845>
   rotation=<rot 15.972 -89.1597 -4.3556>
  }]
 dyn_rope_b01 [{
   shapeType=Capsule
   length=4.4853
   radius=1.8872
   shapeName=dyn_rope_b01_capsule
   center=<-2.4708 0.6714 -5.6875>
   rotation=<rot -6.0494 -90 27.846>
  }]
 dyn_rope_b02 [{shapeType=Capsule, length=7.8987, radius=2.6116, shapeName=dyn_rope_b02_capsule, center=<-3 0 -5>, rotation=<rot 3.196 -90 31.5585>}]
 pack [{shapeType=Capsule, length=25, radius=5, shapeName=pack_capsule, center=<5.62 -8.3576 -8.9361>}]
 pelvis [{shapeType=Capsule, length=9.0386, radius=12.297, shapeName=pelvis_capsule, center=<-1.4149 -0.6973 0>}]
 spine_01 [{shapeType=Capsule, length=10.9044, radius=10.8498, shapeName=spine_01_capsule, center=<4.2664 0 0>}]
 spine_05 [{shapeType=Capsule, length=20.3513, radius=9.6487, shapeName=spine_05_capsule, center=<-9.0957 0 0>}]
 upperarm_r [{
   shapeType=Capsule
   length=24.3268
   radius=5.6732
   shapeName=upperarm_r_capsule
   center=<-18.4208 -0.0001 0>
   rotation=<rot 90 -5.051 -5.0513>
  }]
}
constraints [
 {bone1Name=dyn_rope_a03, bone2Name=dyn_rope_a02, swing1LimitDegrees=5, swing2LimitDegrees=45, twistLimitDegrees=45}
 {
  bone1Name=dyn_Tiger_Head_02
  bone2Name=pack
  swing1LimitDegrees=45
  swing1Motion=Free
  swing2LimitDegrees=45
  swing2Motion=Free
  twistLimitDegrees=45
  twistMotion=Free
 }
 {bone1Name=dyn_rope_a02, bone2Name=dyn_rope_a01, swing1LimitDegrees=5, swing2LimitDegrees=45, twistLimitDegrees=45}
 {bone1Name=dyn_rope_b02, bone2Name=dyn_rope_b01, swing1LimitDegrees=5, swing2LimitDegrees=30, twistLimitDegrees=30}
 {bone1Name=dyn_rope_b01, bone2Name=pack, swing1LimitDegrees=45, swing2LimitDegrees=45, twistLimitDegrees=45}
 {
  bone1Name=dyn_rope_a01
  bone2Name=pack
  swing1LimitDegrees=45
  swing1Motion=Free
  swing2LimitDegrees=45
  swing2Motion=Free
  twistLimitDegrees=45
  twistMotion=Free
 }
]
