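# Anim Blueprint for the CoyoTear backpack cosmetic, adding secondary motion to its tiger head and ropes.
# From ref pose, additive bone offsets feed pack-relative AnimDynamics on head and rope bones, then a RigidBody node using the pack physics asset.
# Before RigidBody, pelvis, spine_01, spine_05 and upperarm_r are replaced with world transforms copied each update from the character mesh.
# Update also computes a state-based gravity override, which the AnimGraph never reads.
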AnimBlueprint /BRCosmetics/Accessories/FORT_Backpacks/Backpack_F_MED_CoyoTear/Meshes/F_MED_CoyoTear_Pack_AnimBP
target_skeleton=&/F_MED_CoyoTear_Pack_Skeleton
graphs [
 AnimGraph {nodes [
   n1 = "Transform (Modify) Bone - Bone: pack"(ComponentPose=Local To Component(LocalPose=Local Space Ref Pose()), Translation="0,0,0", Rotation="0,0,10") {
    rotationMode=BMM_Additive
    rotationSpace=BCS_BoneSpace
    scaleMode=BMM_Additive
    scaleSpace=BCS_BoneSpace
    translationMode=BMM_Additive
    translationSpace=BCS_BoneSpace
    alphaBoolBlend {blendOption=Linear}
    boneToModify {boneName=pack}
    rotation {roll=10}
   }
   n2 = "Transform (Modify) Bone - Bone: dyn_Tiger_Head_01"(ComponentPose=@n1, Alpha=1, Translation="0,0,0", Rotation="0,0,13", Scale="1,1,1") {
    rotationMode=BMM_Additive
    rotationSpace=BCS_BoneSpace
    scaleMode=BMM_Additive
    scaleSpace=BCS_BoneSpace
    translationMode=BMM_Additive
    translationSpace=BCS_BoneSpace
    alphaBoolBlend {blendOption=Linear}
    boneToModify {boneName=dyn_Tiger_Head_01}
    rotation {roll=13}
   }
   n3 = "Transform (Modify) Bone - Bone: dyn_Tiger_Head_02"(ComponentPose=@n2, Alpha=1, Translation="0,0,0", Rotation="0,0,0", Scale="1,1,1") {
    rotationMode=BMM_Additive
    rotationSpace=BCS_BoneSpace
    scaleMode=BMM_Additive
    scaleSpace=BCS_BoneSpace
    translationMode=BMM_Additive
    translationSpace=BCS_BoneSpace
    alphaBoolBlend {blendOption=Linear}
    boneToModify {boneName=dyn_Tiger_Head_02}
   }
   n4 = Anim Dynamics - Bone: dyn_Tiger_Head_01!(ComponentPose=@n3, Alpha=1, GravityOverride:Vector="0,0,-980", ExternalForce="0,0,0") -> (Pose) {
    angularDampingOverride=0.99
    angularSpringConstant=800
    bOverrideAngularDamping=true
    bUseGravityOverride=true
    simulationSpace=BoneRelative
    alphaBoolBlend {blendOption=Linear}
    boundBone {boneName=dyn_Tiger_Head_01}
    componentAppliedLinearAccClamp=<1500 1500 600>
    componentLinearAccScale=<0.4 0.4 0.4>
    componentLinearVelScale=<0.2 0.2 0.2>
    gravityOverride {z=-980}
    physicsBodyDefinitions [{
      collisionType=CustomSphere
      sphereCollisionRadius=2
      boundBone {boneName=dyn_Tiger_Head_01}
      boxExtents=<30 30 30>
      constraintSetup {twistAxis=AxisZ, angularLimitsMax=<20 25 5>, angularLimitsMin=<-35 -25 -5>, angularTarget=<0 0 0>}
      localJointOffset=<0 -2 -5>
     }]
    relativeSpaceBone {boneName=pack}
   }
   n5 = Anim Dynamics - Bone: dyn_Tiger_Head_02!(ComponentPose=@n4, Alpha=1, GravityOverride:Vector="0,0,-980", ExternalForce="0,0,0") -> (Pose) {
    angularDampingOverride=0.99
    angularSpringConstant=800
    bOverrideAngularDamping=true
    bUseGravityOverride=true
    simulationSpace=BoneRelative
    alphaBoolBlend {blendOption=Linear}
    boundBone {boneName=dyn_Tiger_Head_02}
    componentAppliedLinearAccClamp=<1500 1500 600>
    componentLinearAccScale=<0.4 0.4 0.4>
    componentLinearVelScale=<0.2 0.2 0.2>
    gravityOverride {z=-980}
    physicsBodyDefinitions [{
      collisionType=CustomSphere
      sphereCollisionRadius=2
      boundBone {boneName=dyn_Tiger_Head_02}
      boxExtents=<25 25 25>
      constraintSetup {angularLimitsMax=<1 0.5 5>, angularLimitsMin=<-1 -0.5 0>, angularTarget=<0 0 0>}
      localJointOffset=<0 -2 -1>
     }]
    relativeSpaceBone {boneName=pack}
   }
   n6 = Anim Dynamics - Bone: dyn_rope_a01!(ComponentPose=@n5, Alpha=1, GravityOverride:Vector="0,0,-980", ExternalForce="0,0,0") -> (Pose) {
    angularDampingOverride=0.99
    angularSpringConstant=800
    bAngularSpring=true
    bOverrideAngularDamping=true
    bUseGravityOverride=true
    gravityScale=0.3
    simulationSpace=BoneRelative
    alphaBoolBlend {blendOption=Linear}
    boundBone {boneName=dyn_rope_a01}
    componentAppliedLinearAccClamp=<2500 2500 1500>
    componentLinearAccScale=<0.5 0.3 0.4>
    componentLinearVelScale=<0.55 0.4 0.45>
    gravityOverride {z=-980}
    physicsBodyDefinitions [{
      collisionType=CustomSphere
      sphereCollisionRadius=2
      boundBone {boneName=dyn_rope_a01}
      boxExtents=<30 30 30>
      constraintSetup {angularLimitsMax=<15 0 5>, angularLimitsMin=<-5 -20 0>, angularTarget=<0 0 0>}
      localJointOffset=<0.1 0.1 -4>
     }]
    relativeSpaceBone {boneName=pack}
   }
   n7 = Anim Dynamics - Bone: dyn_rope_b01!(ComponentPose=@n6, Alpha=1, GravityOverride:Vector="0,0,-980", ExternalForce="0,0,0") -> (Pose) {
    angularDampingOverride=0.99
    angularSpringConstant=800
    bAngularSpring=true
    bOverrideAngularDamping=true
    bUseGravityOverride=true
    gravityScale=0.3
    simulationSpace=BoneRelative
    alphaBoolBlend {blendOption=Linear}
    boundBone {boneName=dyn_rope_b01}
    componentAppliedLinearAccClamp=<2500 2500 1500>
    componentLinearAccScale=<0.5 0.3 0.4>
    componentLinearVelScale=<0.55 0.4 0.45>
    gravityOverride {z=-980}
    physicsBodyDefinitions [{
      collisionType=CustomSphere
      sphereCollisionRadius=2
      boundBone {boneName=dyn_rope_b01}
      boxExtents=<30 30 30>
      constraintSetup {angularLimitsMax=<5 0 5>, angularLimitsMin=<-5 -20 0>, angularTarget=<0 0 0>}
      localJointOffset=<0.1 0.1 -5>
     }]
    relativeSpaceBone {boneName=pack}
   }
   n8 = "Transform (Modify) Bone - Bone: pelvis"(
    ComponentPose=@n7
    Alpha=1
    Translation=Get pelvis_tran() -> (pelvis_tran:Vector)
    Rotation=Get pelvis_rot() -> (pelvis_rot:Rotator)
    Scale="1,1,1"
   ) {
    rotationMode=BMM_Replace
    rotationSpace=BCS_WorldSpace
    scaleMode=BMM_Additive
    scaleSpace=BCS_BoneSpace
    translationMode=BMM_Replace
    translationSpace=BCS_WorldSpace
    alphaBoolBlend {blendOption=Linear}
    boneToModify {boneName=pelvis}
    rotation {roll=10}
   }
   n9 = "Transform (Modify) Bone - Bone: spine_01"(
    ComponentPose=@n8
    Alpha=1
    Translation=Get spine01_tran() -> (spine01_tran:Vector)
    Rotation=Get spine01_rot() -> (spine01_rot:Rotator)
    Scale="1,1,1"
   ) {
    rotationMode=BMM_Replace
    rotationSpace=BCS_WorldSpace
    scaleMode=BMM_Additive
    scaleSpace=BCS_BoneSpace
    translationMode=BMM_Replace
    translationSpace=BCS_WorldSpace
    alphaBoolBlend {blendOption=Linear}
    boneToModify {boneName=spine_01}
    rotation {roll=10}
   }
   n10 = "Transform (Modify) Bone - Bone: spine_05"(
    ComponentPose=@n9
    Alpha=1
    Translation=Get spine05_tran() -> (spine05_tran:Vector)
    Rotation=Get spine05_rot() -> (spine05_rot:Rotator)
    Scale="1,1,1"
   ) {
    rotationMode=BMM_Replace
    rotationSpace=BCS_WorldSpace
    scaleMode=BMM_Additive
    scaleSpace=BCS_BoneSpace
    translationMode=BMM_Replace
    translationSpace=BCS_WorldSpace
    alphaBoolBlend {blendOption=Linear}
    boneToModify {boneName=spine_05}
    rotation {roll=10}
   }
   n11 = "Transform (Modify) Bone - Bone: upperarm_r"(
    ComponentPose=@n10
    Alpha=1
    Translation=Get upperarm_r_tran() -> (upperarm_r_tran:Vector)
    Rotation=Get upperarm_r_rot() -> (upperarm_r_rot:Rotator)
    Scale="1,1,1"
   ) {
    rotationMode=BMM_Replace
    rotationSpace=BCS_WorldSpace
    scaleMode=BMM_Additive
    scaleSpace=BCS_BoneSpace
    translationMode=BMM_Replace
    translationSpace=BCS_WorldSpace
    alphaBoolBlend {blendOption=Linear}
    boneToModify {boneName=upperarm_r}
    rotation {roll=10}
   }
   n12 = RigidBody!(ComponentPose=@n11, Alpha, OverrideWorldGravity:Vector="0,0,-980", ExternalForce) -> (Pose) {
    bClampLinearTranslationLimitToRefPose=true
    bOverrideWorldGravity=true
    overridePhysicsAsset=&/F_MED_CoyoTear_Pack_Physics
    simulationSpace=BaseBoneSpace
    alphaBoolBlend {blendOption=Linear}
    alphaScaleBiasClamp {bInterpResult=true, clampMax=0.5}
    baseBoneRef {boneName=pack}
    componentAppliedLinearAccClamp=<2000 2000 2000>
    componentLinearAccScale=<0.27 0.2 0.13>
    componentLinearVelScale=<0.3 0.25 0.15>
    overrideWorldGravity {z=-980}
   }
   Output Pose(Result=Component To Local(ComponentPose=@n12)) {layerGroup=None}
  ]}
 EventGraph {nodes [
   Event BlueprintUpdateAnimation()
   > n1 = Gravity Override State Function DEPRECATED(CharacterPartAnimInstance=Self-Reference(), GameTimeInSeconds=GetGameTimeInSeconds(), GravityOverrideParams=Get Gravity Override Params())
   > n2 = Set Gravity Override Cape(Gravity Override Cape:Vector=@n1) -> (Output_Get:Vector)
   n3 = Get Character Mesh() -> (Character Mesh:Skeletal Mesh Component Object Reference)
   n4 = BreakTransform(InTransform=GetSocketTransform(self=@n3, InSocketName=root, TransformSpace=RTS_World))
   Set root_tran(execute=[@n12.Is Valid, @n2.then], root_tran:Vector=@n4) -> (Output_Get:Vector)
   > Set root_rot(root_rot:Rotator=@n4.Rotation) -> (Output_Get:Rotator)
   n5 = BreakTransform(InTransform=GetSocketTransform(self=@n3, InSocketName=pelvis, TransformSpace=RTS_World))
   > Set pelvis_tran(pelvis_tran:Vector=@n5) -> (Output_Get:Vector)
   > Set pelvis_rot(pelvis_rot:Rotator=@n5.Rotation) -> (Output_Get:Rotator)
   n6 = BreakTransform(InTransform=GetSocketTransform(self=@n3, InSocketName=spine_01, TransformSpace=RTS_World))
   > Set spine01_tran(spine01_tran:Vector=@n6) -> (Output_Get:Vector)
   > Set spine01_rot(spine01_rot:Rotator=@n6.Rotation) -> (Output_Get:Rotator)
   n7 = BreakTransform(InTransform=GetSocketTransform(self=@n3, InSocketName=spine_05, TransformSpace=RTS_World))
   > Set spine05_tran(spine05_tran:Vector=@n7) -> (Output_Get:Vector)
   > Set spine05_rot(spine05_rot:Rotator=@n7.Rotation) -> (Output_Get:Rotator)
   n8 = BreakTransform(InTransform=GetSocketTransform(self=@n3, InSocketName=upperarm_r, TransformSpace=RTS_World))
   > Set upperarm_r_tran(upperarm_r_tran:Vector=@n8) -> (Output_Get:Vector)
   > Set upperarm_r_rot(upperarm_r_rot:Rotator=@n8.Rotation) -> (Output_Get:Rotator)
   n9 = Event OnCharacterCustomizationLoaded()
   > n10 = Sequence()
   > Set Character Mesh(Character Mesh:Skeletal Mesh Component Object Reference=@n9.CustomBaseMesh) -> (Output_Get:Skeletal Mesh Component Object Reference)
   Set Body Mesh(
    execute=@n10.then_1
    Body Mesh:Skeletal Mesh Component Object Reference=Get Array Item with EFort Custom Part Type(Array=@n9.CustomMeshParts)
   ) -> (Output_Get:Skeletal Mesh Component Object Reference)
   n11 = GetPartSkeletalMeshComponent()
   Is Valid(InputObject=@n11)
   > n12 = Is Valid(InputObject=GetAttachParent!(self:Scene Component Object Reference=@n11) -> (ReturnValue:Scene Component Object Reference))
   GetSkeletalMesh!(self:Custom Character Part Object Reference=Get Array Item with EFort Custom Part Type(Array:Array of Custom Character Part Object References=@n9.CustomParts) -> (Output:Custom Character Part Object Reference)) -> (ReturnValue:Skeletal Mesh Object Reference)
  ]}
]
variables [
 Body Mesh {type=SkeletalMeshComponent}
 Character Mesh {type=SkeletalMeshComponent}
 clavicle_r_rot {type=Rotator}
 clavicle_r_tran {type=Vector}
 Gravity Override Cape {type=Vector}
 Gravity Override Params {
  type=GravityOverrideParamsStruct
  default {
   dBNO {jointName=pack, gravityOverride=<0 0 -980>}
   dBNO_Carried {jointName=pack, gravityOverride=<0 0 -980>}
   riding_Moving {gravityOverride=<0 -450 -980>}
   riding_Standing {gravityOverride=<0 -200 -980>}
   skydive_DivingDown {jointName=pack, gravityOverride=<0 0 0>, windAmplitude=<0 0 0>, windFrequency=<0 0 0>}
   skydive_FloatingDown {jointName=pack, gravityOverride=<0 0 -150>, windAmplitude=<0 0 0>, windFrequency=<0 0 0>, windOffset=<0 0 0>}
   skydive_GoingUp {gravityOverride=<0 -100 -200>}
   skydive_Parachuting {gravityOverride=<0 -600 -700>}
   swimming {jointName=pack}
   tactical_Sprint {gravityOverride=<0 -450 -980>}
  }
 }
 pelvis_rot {type=Rotator}
 pelvis_tran {type=Vector}
 root_rot {type=Rotator}
 root_tran {type=Vector}
 spine01_rot {type=Rotator}
 spine01_tran {type=Vector}
 spine02_rot {type=Rotator}
 spine02_tran {type=Vector}
 spine03_rot {type=Rotator}
 spine03_tran {type=Vector}
 spine04_rot {type=Rotator}
 spine04_tran {type=Vector}
 spine05_rot {type=Rotator}
 spine05_tran {type=Vector}
 upperarm_r_rot {type=Rotator}
 upperarm_r_tran {type=Vector}
]
